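# Control Rig for the face accessory part of the female SurgeRaven Zany character, correcting the ponytail base as the head moves.
# A spherical pose reader measures head orientation against spine_04 within an asymmetric falloff region, with debug drawing left on.
# Its output weights a Modify Transforms node that rotates and slightly shifts base_ponytail.
# The hierarchy also holds bangs and dyn_ ponytail bones, which the graph does not drive directly.
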ControlRigBlueprint /BRCosmetics/Characters/Player/Female/Medium/Bodies/F_MED_SurgeRaven_Zany/Meshes/Parts/F_MED_SurgeRaven_Zany_FaceAcc_CtrlRig
graphs [
 RigVMFunctionLibrary {links [], nodes []}
 RigVMModel {nodes [
   RigUnit_BeginExecution = &RigUnit_BeginExecution()
   n1 = &RigUnit_SphericalPoseReader(
    ActiveRegionScaleFactors=(PositiveWidth=3,NegativeWidth=0.3,PositiveHeight=6,NegativeHeight=6)
    ActiveRegionSize=0.1
    Debug=(bDrawDebug=True,DebugScale=25,DebugSegments=20,DebugThickness=0.25)
    DriverAxis=(X=1,Y=0,Z=0)
    DriverItem=(Type=Bone,Name="head")
    FalloffRegionScaleFactors=(PositiveWidth=-3,NegativeWidth=1,PositiveHeight=0,NegativeHeight=0)
    FalloffSize=0.25
    OptionalParentItem=(Type=Bone,Name="spine_04")
    RotationOffset=(X=90,Y=0,Z=-140)
    ExecutePin=@RigUnit_BeginExecution.ExecutePin
   )
   > &RigUnit_ModifyTransforms(
    ItemToModify=((Item=(Type=Bone,Name="base_ponytail"),Transform=(Rotation=(X=0,Y=0,Z=0.342,W=0.9397),Translation=(X=-0.5,Y=0,Z=0),Scale3D=(X=1,Y=1,Z=1))))
    Weight=@n1.OutputParam
   )
  ]}
]
hierarchy [
 root {loc=<0 0 0>, rot=<0 0 0 1>}
 pelvis {parent=root, loc=<0 -0.5301 98.0571>, rot=<0.0127 0.707 -0.0127 -0.707>}
 spine_01 {parent=pelvis, loc=<6.1208 -0.9328 0>, rot=<0 0 0.034 -0.9994>}
 spine_02 {parent=spine_01, loc=<6.5997 -0.2095 0>, rot=<0 0 0.0238 -0.9997>}
 spine_03 {parent=spine_02, loc=<6.4914 -0.1237 0>, rot=<0 0 0.0149 -0.9999>}
 spine_04 {parent=spine_03, loc=<6.9598 0.0794 0>, rot=<0 0 -0.0132 -0.9999>}
 spine_05 {parent=spine_04, loc=<6.9987 0 0>, rot=<0 0 -0.0015 -1>}
 neck_01 {parent=spine_05, loc=<11.3138 0.1529 0>, rot=<0 0 -0.2003 -0.9797>}
 neck_02 {parent=neck_01, loc=<4.5627 0 0>, rot=<0 0 0.0504 -0.9987>}
 head {parent=neck_02, loc=<5.3914 0 0>, rot=<0 0 0.115 -0.9934>}
 base_bangs_sml_outer_r {parent=head, loc=<9.1045 6.0631 5.4341>, rot=<0 0 0.999 -0.0436>}
 dyn_bangs_sml_outer_r {parent=base_bangs_sml_outer_r, loc=<0 0 0>, rot=<0 0 0 1>}
 base_bangs_lrg_r {parent=head, loc=<6.1574 4.0003 6.6609>, rot=<0 0 0.999 -0.0436>}
 dyn_bangs_lrg_r {parent=base_bangs_lrg_r, loc=<0 0 0>, rot=<0 0 0 1>}
 dyn_bangs_lrg_01_r {parent=dyn_bangs_lrg_r, loc=<3.301 2.2173 0.0968>, rot=<0 0 0.0436 -0.999>}
 base_bangs_lrg_l {parent=head, loc=<6.7863 4.6599 -5.9866>, rot=<0 0 0.999 -0.0436>}
 dyn_bangs_lrg_l {parent=base_bangs_lrg_l, loc=<0 0 0>, rot=<0 0 0 1>}
 dyn_bangs_lrg_01_l {parent=dyn_bangs_lrg_l, loc=<3.657 1.2787 -0.6128>, rot=<0 0 0 1>}
 base_ponytail {parent=head, loc=<12.2349 -7.374 0.0417>, rot=<0 0 0.6428 -0.766>}
 dyn_ponytail {parent=base_ponytail, loc=<0 0 0>, rot=<0 0 0 1>}
 dyn_ponytail_01 {parent=dyn_ponytail, loc=<3.1143 1.004 0>, rot=<0 0 0.766 -0.6428>}
 dyn_ponytail_02 {parent=dyn_ponytail_01, loc=<7.0912 2.3071 0>, rot=<0 0 0.0436 -0.999>}
 dyn_ponytail_end_l {parent=dyn_ponytail_02, loc=<7.7691 -2.327 -3.3672>, rot=<0 0 0 1>}
 dyn_ponytail_end_r {parent=dyn_ponytail_02, loc=<6.0468 -0.0597 3.1451>, rot=<0 0 0 1>}
 dyn_ponytail_end_m {parent=dyn_ponytail_02, loc=<5.124 -0.3564 0>, rot=<0 0 0 1>}
 dyn_ponytail_end_ext_m {parent=dyn_ponytail_end_m, loc=<6.3295 -1.6861 0.3435>, rot=<0 0 -0.0872 -0.9962>}
 dyn_ponytail_end_ext_l {parent=dyn_ponytail_end_m, loc=<5.0799 -2.4266 -1.5682>, rot=<0 0 -0.0872 -0.9962>}
 dyn_ponytail_end_ext_r {parent=dyn_ponytail_end_m, loc=<4.7489 -2.03 2.4741>, rot=<0 0 -0.0872 -0.9962>}
 base_bangs_sml_outer_l {parent=head, loc=<9.0959 7.0989 -5.2951>, rot=<0 0 0.999 -0.0436>}
 dyn_bangs_sml_outer_l {parent=base_bangs_sml_outer_l, loc=<0 0 0>, rot=<0 0 0 1>}
 base_bangs_sml_inner_r {parent=head, loc=<10.8854 7.5648 2.985>, rot=<0 0 0.999 -0.0436>}
 dyn_bangs_sml_inner_r {parent=base_bangs_sml_inner_r, loc=<0 0 0>, rot=<0 0 0 1>}
 base_bangs_sml_inner_l {parent=head, loc=<10.8854 7.5648 -2.9826>, rot=<0 0 0.999 -0.0436>}
 dyn_bangs_sml_inner_l {parent=base_bangs_sml_inner_l, loc=<0 0 0>, rot=<0 0 0 1>}
 base_bangs_sml_m {parent=head, loc=<11.5989 7.5707 0.2127>, rot=<0 0 0.999 -0.0436>}
 dyn_bangs_sml_m {parent=base_bangs_sml_m, loc=<0 0 0>, rot=<0 0 0 1>}
]
