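# Physics asset for the SurgeRaven Zany female medium FaceAcc mesh, giving its ponytail jointed, limited-motion physics.
# Capsules on the head, neck, upper spine and clavicles form body volumes linked by free angular constraints.
# The ponytail base locks rotationally to the head and leads into a two-bone chain with tight swing and twist limits.
# That chain forks into left, middle and right ends, with the middle splitting again into three extension tips.
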PhysicsAsset /BRCosmetics/Characters/Player/Female/Medium/Bodies/F_MED_SurgeRaven_Zany/Meshes/Parts/F_MED_SurgeRaven_Zany_FaceAcc_Physics
bodies {
 base_ponytail [{shapeType=Capsule, length=0.5, radius=0.5, shapeName=base_ponytail_capsule, center=<3 0 0>, rotation=<rot 90 0 0>}]
 clavicle_l [{shapeType=Capsule, length=6, radius=6, shapeName=clavicle_l_capsule, center=<6 0 0>, rotation=<rot 90 0 0>}]
 clavicle_r [{shapeType=Capsule, length=6, radius=6, shapeName=clavicle_r_capsule, center=<-6 0 0>, rotation=<rot 90 0 0>}]
 dyn_ponytail_01 [{shapeType=Capsule, length=3, radius=3, shapeName=dyn_ponytail_01_capsule, center=<2.23 1.1523 0>, rotation=<rot 90 33.7499 0>}]
 dyn_ponytail_02 [{shapeType=Capsule, length=3, radius=3, shapeName=dyn_ponytail_02_capsule, center=<2.0599 0 0>, rotation=<rot 90 0 0>}]
 dyn_ponytail_end_ext_l [{
   shapeType=Capsule
   length=2.5
   radius=0.8
   shapeName=dyn_ponytail_end_ext_l_capsule
   center=<2.4136 0.3036 -0.1865>
   rotation=<rot 90 -2.8125 0>
  }]
 dyn_ponytail_end_ext_m [{shapeType=Capsule, length=2.5, radius=0.8, shapeName=dyn_ponytail_end_ext_m_capsule, center=<2.0028 -0.0667 0>, rotation=<rot 90 -2.8125 0>}]
 dyn_ponytail_end_ext_r [{
   shapeType=Capsule
   length=2.5
   radius=1
   shapeName=dyn_ponytail_end_ext_r_capsule
   center=<2.3954 -0.0667 0.1867>
   rotation=<rot 90 -2.8125 0>
  }]
 dyn_ponytail_end_l [{shapeType=Capsule, length=1.8, radius=2.5, shapeName=dyn_ponytail_end_l_capsule, center=<2.149 0 -0.473>, rotation=<rot 90 0 0>}]
 dyn_ponytail_end_m [{shapeType=Capsule, length=6, radius=3.5, shapeName=dyn_ponytail_end_m_capsule, center=<3.7918 -0.4246 0>, rotation=<rot 90 0 0>}]
 dyn_ponytail_end_r [{shapeType=Capsule, length=2, radius=2.5, shapeName=dyn_ponytail_end_r_capsule, center=<2.8752 0 0.5475>, rotation=<rot 90 0 0>}]
 head [{shapeType=Capsule, length=0, radius=9, shapeName=head_capsule, center=<4.3577 -0.2245 0>}]
 neck_01 [{shapeType=Capsule, length=5, radius=5, shapeName=neck_01_capsule, rotation=<rot 90 0 0>}]
 neck_02 [{shapeType=Capsule, length=5, radius=5, shapeName=neck_02_capsule, rotation=<rot 90 0 0>}]
 spine_05 [
  {shapeType=Capsule, length=18, radius=11, shapeName=spine_05_capsule, center=<-0.0347 -1.7689 0>, rotation=<rot 0 -90 0>}
  {shapeType=Capsule, length=11, radius=5, shapeName=spine_05_capsule, center=<14.68 -0.0038 0>}
 ]
}
constraints [
 {
  bone1Name=head
  bone2Name=neck_02
  swing1LimitDegrees=45
  swing1Motion=Free
  swing2LimitDegrees=45
  swing2Motion=Free
  twistLimitDegrees=45
  twistMotion=Free
 }
 {
  bone1Name=neck_02
  bone2Name=neck_01
  swing1LimitDegrees=45
  swing1Motion=Free
  swing2LimitDegrees=45
  swing2Motion=Free
  twistLimitDegrees=45
  twistMotion=Free
 }
 {
  bone1Name=neck_01
  bone2Name=spine_05
  swing1LimitDegrees=45
  swing1Motion=Free
  swing2LimitDegrees=45
  swing2Motion=Free
  twistLimitDegrees=45
  twistMotion=Free
 }
 {
  bone1Name=base_ponytail
  bone2Name=head
  swing1LimitDegrees=45
  swing1Motion=Locked
  swing2LimitDegrees=45
  swing2Motion=Locked
  twistLimitDegrees=45
  twistMotion=Locked
 }
 {bone1Name=dyn_ponytail_02, bone2Name=dyn_ponytail_01, swing1LimitDegrees=20, swing2LimitDegrees=25, twistLimitDegrees=5}
 {bone1Name=dyn_ponytail_01, bone2Name=base_ponytail, swing1LimitDegrees=20, swing2LimitDegrees=25, twistLimitDegrees=5}
 {bone1Name=dyn_ponytail_end_l, bone2Name=dyn_ponytail_02, swing1LimitDegrees=10, swing2LimitDegrees=10, twistLimitDegrees=5}
 {bone1Name=dyn_ponytail_end_m, bone2Name=dyn_ponytail_02, swing1LimitDegrees=25, swing2LimitDegrees=30, twistLimitDegrees=5}
 {bone1Name=dyn_ponytail_end_ext_l, bone2Name=dyn_ponytail_end_m, swing1LimitDegrees=30, swing2LimitDegrees=20, twistLimitDegrees=5}
 {bone1Name=dyn_ponytail_end_ext_m, bone2Name=dyn_ponytail_end_m, swing1LimitDegrees=30, swing2LimitDegrees=20, twistLimitDegrees=5}
 {bone1Name=dyn_ponytail_end_ext_r, bone2Name=dyn_ponytail_end_m, swing1LimitDegrees=30, swing2LimitDegrees=20, twistLimitDegrees=5}
 {bone1Name=dyn_ponytail_end_r, bone2Name=dyn_ponytail_02, swing1LimitDegrees=10, swing2LimitDegrees=10, twistLimitDegrees=5}
 {
  bone1Name=clavicle_l
  bone2Name=spine_05
  swing1LimitDegrees=45
  swing1Motion=Free
  swing2LimitDegrees=45
  swing2Motion=Free
  twistLimitDegrees=45
  twistMotion=Free
 }
 {
  bone1Name=clavicle_r
  bone2Name=spine_05
  swing1LimitDegrees=45
  swing1Motion=Free
  swing2LimitDegrees=45
  swing2Motion=Free
  twistLimitDegrees=45
  twistMotion=Free
 }
]
